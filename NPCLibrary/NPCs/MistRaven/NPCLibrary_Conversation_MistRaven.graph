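# Conversation database for the MistRaven NPC, Morgan Myst of The Supernovas, serving as the greeting and top-level hub when a player talks to them.
# The entry tag leads to one Speech node that picks its line by context and priority: first-ever meeting, match conversation count, or the player's equipped character.
# Character-specific lines carry raised priority over the count-based ones.
# The speech then branches through link nodes to the separate Quests, BasicServices and Dialogue entry points.
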ConversationDatabase /NPCLibrary/NPCs/MistRaven/NPCLibrary_Conversation_MistRaven
entry_points [{tag=Conversation.Entry.MistRaven, destinations [#1]}]
nodes [
 ConversationEntryPointNode_7 {
  id=#1
  class=ConversationEntryPointNode
  outputs [#2]
  properties {entryTag {tagName=Conversation.Entry.MistRaven}}
 }
 Speech_C_4 {
  id=#2
  class=Speech_C
  outputs [#3, #4, #5]
  properties {
   nodeName=""
   generalConfig {
    defaultMessage=My Supernova Academy training didn't exactly cover swarms of Bugs.
    contextualMessages [
     {
      message="I'm Morgan Myst, team leader of The Supernovas. How can I help?"
      contextRequirements [{requirement {class=FortControllerRequirement_IsFirstEverConversationWithNPC}}]
      priority {curve {rowName=Tandem.Speech.Priority.FirstEverInteraction}}
     }
     {
      message=My Supernova Academy training didn't exactly cover swarms of Bugs.
      contextRequirements [{requirement {class=FortControllerRequirement_CheckNPCMatchConversationCount, properties {nPCUniqueIdTag {tagName=AISpawnerData.Type.Tandem.MistRaven}}}}]
      priority {curve {curveTable=null, rowName=None}}
     }
     {
      message="We have to protect the Island, before all of this gets worse."
      contextRequirements [{requirement {class=FortControllerRequirement_CheckNPCMatchConversationCount, properties {nPCUniqueIdTag {tagName=AISpawnerData.Type.Tandem.MistRaven}}}}]
      priority {curve {curveTable=null, rowName=None}}
     }
     {
      message=The Supernovas are supporting O.X.R. in protecting the Island.
      contextRequirements [{requirement {class=FortControllerRequirement_CheckNPCMatchConversationCount, properties {nPCUniqueIdTag {tagName=AISpawnerData.Type.Tandem.MistRaven}}}}]
      priority {curve {curveTable=null, rowName=None}}
     }
     {
      message=Did you... see this coming?
      contextRequirements [{requirement {
         class=FortControllerRequirement_HasCID
         properties {allowedCIds [
           &Character_IronLilac_Sly
           &Character_BrightDisk
           &Character_VoidRedemption_Rebel
          ]}
        }}]
      priority {value=2, curve {curveTable=null, rowName=None}}
     }
     {
      message="Time for you to step up and help, instead of trying to be the hero."
      contextRequirements [{requirement {
         class=FortControllerRequirement_HasCID
         properties {allowedCIds [&Character_MistMylar_Claw, /BRCosmetics/Athena/Items/Cosmetics/Characters/Character_SolidMist]}
        }}]
      priority {value=2, curve {curveTable=null, rowName=None}}
     }
     {
      message=We're The Supernovas. It's our job to protect this Island. So let's do it.
      requirementMatchPolicy=RequireAny
      contextRequirements [{requirement {
         class=FortControllerRequirement_HasCID
         properties {allowedCIds [
           /BRCosmetics/Athena/Items/Cosmetics/Characters/Character_LazyLizzChip
           /BRCosmetics/Athena/Items/Cosmetics/Characters/Character_AshenMagnusSlant
           /BRCosmetics/Athena/Items/Cosmetics/Characters/Character_SpatialTravelWatt
           /BRCosmetics/Athena/Items/Cosmetics/Characters/Character_BentBaton
           /BRCosmetics/Athena/Items/Cosmetics/Characters/Character_PolarityWinnField
          ]}
        }}]
      priority {value=2, curve {curveTable=null, rowName=None}}
     }
     {
      message=The Supernovas will make you proud. I promise.
      requirementMatchPolicy=RequireAny
      contextRequirements [{requirement {
         class=FortControllerRequirement_HasCID
         properties {allowedCIds [
           /BRCosmetics/Athena/Items/Cosmetics/Characters/Character_SwoopClaspPry
           &Character_UndergroundRebel_Fashion
           &CID_989_Athena_Commando_M_ProgressiveJonesy
           &Character_RoseDepth_Seed
          ]}
        }}]
      priority {value=2, curve {curveTable=null, rowName=None}}
     }
    ]
   }
  }
 }
 ConversationLinkNode_3 {id=#5, class=ConversationLinkNode, properties {remoteEntryTag {tagName=Conversation.Entry.MistRaven.Dialogue}}}
 ConversationLinkNode_0 {id=#4, class=ConversationLinkNode, properties {remoteEntryTag {tagName=Conversation.Entry.BasicServices}}}
 ConversationLinkNode_1 {id=#3, class=ConversationLinkNode, properties {remoteEntryTag {tagName=Conversation.Entry.MistRaven.Quests}}}
]
